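# Translucent unlit Niagara sprite material for the KneelReedy emote's dark flash, turning the scene behind it into a high-contrast black-and-white image.
# Desaturated scene color times world-normal length is thresholded, with dynamic parameters controlling inversion and blend over scene color.
# A mask keeps pixels nearer than a particle-driven camera distance plus offset, optionally subtracting a custom-stencil match and applying a sphere falloff.
# Shading path switches zero the effect on forward and mobile.
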Material /BRCosmetics/Animation/Game/MainPlayer/Emotes/KneelReedy/FX/M_Emote_ArcticIceBlue_Dark_Flash
graph {
 properties {bUsedWithNiagaraSprites=true, blendMode=BLEND_Translucent, pixelDepthOffsetMode=PDOM_Legacy, shadingModel=MSM_Unlit, usedShadingModels=Unlit}
 nodes [
  n1 = ParticleColor()
  n2 = ScalarParameter() {parameterName=Use particle color}
  n3 = Desaturation(None=SceneColor!(UVs).RGB)
  n4 = SceneTexture() {sceneTextureId=PPI_WorldNormal}
  n5 = MaterialFunctionCall(Vector 3=Multiply(A=ComponentMask(None=@n4.Color) {b=true, g=true, r=true}, B=ScalarParameter() {defaultValue=5, parameterName=WorldNormal}) {constB=5}) {materialFunction=/Engine/Functions/Engine_MaterialFunctions02/Utility/VectorLength}
  n6 = Saturate(None=Step(Y=Multiply(A=@n3, B=@n5.V3 Length)) {constX=0.5})
  n7 = LinearInterpolate(
   A=SceneColor().RGB
   B=LinearInterpolate(A=@n6, B=OneMinus(None=@n6), Alpha=DynamicParameter() {paramNames [Body Des, Dark lerp, Param3, Param4]}.Dark lerp)
   Alpha=DynamicParameter() {paramNames [Body Des, Dark lerp, Param3, Param4]}.Body Des
  ) {constB=0}
  n8 = Constant()
  MaterialFunctionCall(Emissive=Multiply(A=LinearInterpolate(B=@n1.RGB, Alpha=@n2) {constA=1}, B=ShadingPathSwitch(Default=@n7, Deferred=@n7, Forward=@n8, Mobile=@n8))) {materialFunction=&MF_FX_EmissiveAdjustments}
  n9 = ScreenPosition()
  n10 = SceneTexture(UVs=@n9.ViewportUV) {sceneTextureId=PPI_CustomDepth}
  n11 = SceneTexture(UVs=@n9.ViewportUV) {sceneTextureId=PPI_SceneDepth}
  n12 = ScalarParameter() {parameterName=Use SceneDepth}
  n13 = DynamicParameter() {parameterIndex=1, paramNames [R, G, B, ""]}
  n14 = MaterialFunctionCall(R=@n13.R, G=@n13.R, B=@n13.R) {materialFunction=&AppendMany}
  n15 = CameraPositionWS()
  n16 = Constant()
  n17 = If(
   A=ComponentMask(None=LinearInterpolate(A=@n10.Color, B=@n11.Color, Alpha=@n12)) {b=true, g=true, r=true}
   B=Add(A=Distance(A=@n14.RGB, B=@n15), B=ScalarParameter() {defaultValue=15, parameterName=FX - Max Distance Offset})
   "A > B"=@n16
   "A == B"=@n16
   "A < B"=Constant() {r=1}
  )
  n18 = SceneTexture() {sceneTextureId=PPI_CustomStencil}
  n19 = Constant()
  n20 = Constant() {r=1}
  n21 = LinearInterpolate(
   A=@n17
   B=Subtract(A=@n17, B=If(A=@n18.Color, "A > B"=@n19, "A == B"=@n20, "A < B"=@n19) {constB=5})
   Alpha=ScalarParameter() {parameterName=MaskProps}
  )
  n22 = Constant()
  n23 = Multiply(
   A=SphereMask(A=TextureCoordinate(), B=Constant2Vector() {g=0.5, r=0.5}, Radius=Constant() {r=0.5}) {hardnessPercent=0}
   B=ScalarParameter() {defaultValue=5, parameterName=SphereMask edge intensity}
  ) {constB=5}
  n24 = Multiply(
   A=ShadingPathSwitch(Default=@n21, Deferred=@n21, Forward=@n22, Mobile=@n22)
   B=LinearInterpolate(B=Saturate(None=@n23), Alpha=ScalarParameter() {parameterName=Use spheremask}) {constA=1}
  )
  n25 = Constant()
  Multiply(
   A=ShadingPathSwitch(Default=@n24, Deferred=@n24, Forward=@n25, Mobile=@n25)
   B=LinearInterpolate(B=ParticleColor().A, Alpha=ScalarParameter() {parameterName=Use particle color}) {constA=1}
  )
  ScalarParameter() {defaultValue=1000, parameterName=MaxDe}
 ]
}
