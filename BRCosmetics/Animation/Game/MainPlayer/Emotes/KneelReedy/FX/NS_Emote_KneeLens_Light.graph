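# Niagara effect for the KneeLens emote in the KneelReedy folder, showing a short-lived mesh-based light glow.
# YellowLightMesh bursts two mesh particles with a bright cyan HDR color, despite the emitter's name.
# Particle partitioning masks the initial rotation so the two meshes face half a turn apart.
# An alpha curve fades them in, holds, then fades out over their brief life; the Empty emitter renders nothing.
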NiagaraSystem /BRCosmetics/Animation/Game/MainPlayer/Emotes/KneelReedy/FX/NS_Emote_KneeLens_Light
dependencies {
 usedDataInterfaces [&NiagaraDataInterfaceCurve]
 usedDynamicInputs [&MaskFloatBySpawnGroup, &MaskVectorBySpawnGroup]
 usedModules [&SystemState]
}
emitters {
 Empty {rendererClasses [], renderers []}
 YellowLightMesh {
  emitterUpdateScript [
   &EmitterState()
   &SpawnBurst_Instantaneous(Spawn Count=2, Use Spawn Probability=false, UseLoopCountLimit=false)
  ]
  particleSpawnScript [
   &InitializeParticle(
    Color=<rgba 0.3131 1.712 2 1>
    Lifetime=1.2
    Mesh Scale Mode=::NewEnumerator3
    Sprite Size=<20 50>
    Sprite Size Mode=::NewEnumerator4
    UsePositionOffset=false
    Write Material Random=true
   )
   &PartitionParticles()
   &InitialMeshOrientation(
    Mesh Orientation Mode=::NewEnumerator1
    Rotation=&MaskVector(Mask Source=1, Mask Parameter=$Particles.PartitionParticles.Partition, Mask=0, Masked Vector=<0 0 0>, Passthrough Vector=<0 0 0.5>)
    Use Rotation Vector=true
   )
  ]
  particleUpdateScript [
   &ParticleState()
   &ScaleColor(
    Scale Alpha=&FloatFromCurve(FloatCurve=DataInterface() {propertyValues {Curve {keys [<at 0.0053 0>,<at 0.1868 1>,<at 0.8194 1>,<at 1 0>]}}})
    ScaleA=true
    ScaleRGB=true
   )
  ]
  rendererClasses [&NiagaraMeshRendererProperties]
  renderers [{rendererClass=&NiagaraMeshRendererProperties, rendererIndex=0}]
 }
}
summary {
 systemName=NS_Emote_KneeLens_Light
 emitters [{emitterName=YellowLightMesh, rendererClasses [&NiagaraMeshRendererProperties]}, {emitterName=Empty, rendererClasses []}]
 userVariables []
}
