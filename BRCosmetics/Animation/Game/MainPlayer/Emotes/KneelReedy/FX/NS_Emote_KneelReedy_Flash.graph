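# Niagara flash effect for the KneelReedy emote, covering a character's body in short-lived sprites.
# Dark_flash and _Frontend_Dark_flash burst one sprite per listed bone of the skeletal mesh, switching between gameplay and front end through FN_FrontEndFloatModifier.
# Both animate dynamic material parameters and fade alpha, with Dark_flash also passing User FX Position to the material.
# A Minimal emitter adds a growing, fading cyan burst, and a Debug variant is disabled.
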NiagaraSystem /BRCosmetics/Animation/Game/MainPlayer/Emotes/KneelReedy/FX/NS_Emote_KneelReedy_Flash
dependencies {
 usedDataInterfaces [
  &NiagaraDataInterfaceCurve
  &NiagaraDataInterfaceArrayFloat
  &NiagaraDataInterfaceSkeletalMesh
  &NiagaraDataInterfaceArrayFloat2
 ]
 usedDynamicInputs []
 usedModules [&SystemState]
}
emitters {
 Dark_flash {
  emitterUpdateScript [
   &EmitterState(Life Cycle Mode=::NewEnumerator1, Loop Behavior=::NewEnumerator1, Loop Duration=1, UseLoopDelay=false)
   &SpawnBurst_Instantaneous(
    Spawn Count=&MakeIntFromFloat(Float=&FN_FrontEndFloatModifier(Gameplay=13, Front End=0))
    Spawn Time=0.1
    Use Spawn Probability=false
    UseLoopCountLimit=false
   )
  ]
  particleSpawnScript [
   &FN_SetSpawnGroup(Number of Spawn Groups=13)
   &InitializeParticle(
    Color Mode=::NewEnumerator0
    Lifetime=0.2
    Sprite Size Mode=::NewEnumerator3
    Uniform Sprite Size=&SelectFloatFromArray(
     Direct Array Index=$Particles.SpawnGroup
     Float Selection Array=DataInterface() {propertyValues {FloatData [40,40,40,40,40,40,40,40,40,40,40,40,40]}}
    )
    UsePositionOffset=false
    Write Material Random=true
   )
   &CameraOffset(Camera Offset Amount=50)
   &SystemLocation(MaskSpawnGroup=false, Offset=<0 0 150>)
  ]
  particleUpdateScript [
   &SkeletalMeshLocation(
    Bone / Socket Index=$Particles.SpawnGroup
    Bone Sampling Mode=::NewEnumerator4
    MaskSpawnGroup=false
    Skeletal Mesh=DataInterface() {propertyValues {
      DefaultMesh=null
      ExcludeBoneName=root
      OptionalComponentClass=null
      PreviewMesh=&M_Med_Soldier_04_Preview
      SkinningMode=SkinOnTheFly
      SoftSourceActor=null
      SourceMode=Default
      UvSetIndex=0
      WholeMeshLOD=-1
      bExcludeBone=false
      bReadDeformedGeometry=true
      bRequireCurrentFrameData=true
      ComponentTags []
      FilteredBones [head,upperarm_l,upperarm_r,lowerarm_l,lowerarm_r,hand_l,hand_r,thigh_l,thigh_r,calf_l,calf_r,foot_l,foot_r]
      FilteredSockets []
      MeshUserParameter {parameter {name=None, typeDefHandle {registeredTypeIndex=1}, varData []}}
      SamplingRegions []
     }}
   )
   &ParticleState()
   &ScaleSpriteSize(Scale Sprite Size Mode=::NewEnumerator0, Uniform Scale Factor=1.15)
   &ScaleColor(
    Scale Alpha=&FloatFromCurve(FloatCurve=DataInterface() {propertyValues {Curve {keys [<at 0 1>,<at 0.5 1>,<at 1 0>]}}})
    ScaleA=true
    ScaleRGB=true
   )
   &DynamicMaterialParameters(
    Index 0 Param 1=1
    Index 0 Param 2=&FloatFromCurve(FloatCurve=DataInterface() {propertyValues {Curve {keys [<at 0 0 RCIM_Constant>, {tangentMode=RCTM_User, time=0.7, value=1}]}}})
    Index 1 Param 1=&MakeFloatFromVector(Vector=$User.FX Position, Channel=::NewEnumerator0)
    Index 1 Param 2=&MakeFloatFromVector(Vector=$User.FX Position, Channel=::NewEnumerator1)
    Index 1 Param 3=&MakeFloatFromVector(Vector=$User.FX Position, Channel=::NewEnumerator2)
    Module.Param0WriteEnabled=true
    Module.Param0WriteEnabled1=true
    Module.Param1WriteEnabled=true
    Module.Param1WriteEnabled1=true
    Module.Param2WriteEnabled=true
    Module.Param2WriteEnabled1=true
    Module.Param3WriteEnabled=true
    Module.Param3WriteEnabled1=true
    Write Parameter Index 1=true
   )
  ]
  rendererClasses [&NiagaraSpriteRendererProperties]
  renderers [
   {rendererClass=&NiagaraSpriteRendererProperties}
   {rendererClass=&NiagaraSpriteRendererProperties, rendererIndex=1}
  ]
 }
 Debug {
  bEnabled=false
  emitterUpdateScript [
   &EmitterState(Life Cycle Mode=::NewEnumerator1, Loop Behavior=::NewEnumerator1, Loop Duration Mode=::NewEnumerator1, UseLoopDelay=false)
   &SpawnRate(SpawnRate=5, Use Spawn Probability=false)
   &SpawnBurst_Instantaneous(
    Spawn Count=&MakeIntFromFloat(Float=&FN_FrontEndFloatModifier(Gameplay=13, Front End=0))
    Spawn Time=0.1
    Use Spawn Probability=false
    UseLoopCountLimit=false
   )
  ]
  particleSpawnScript [
   &FN_SetSpawnGroup(Number of Spawn Groups=13)
   &InitializeParticle(
    Color Mode=::NewEnumerator0
    Lifetime=0.2
    Sprite Size Mode=::NewEnumerator3
    Uniform Sprite Size=&SelectFloatFromArray(
     Direct Array Index=$Particles.SpawnGroup
     Float Selection Array=DataInterface() {propertyValues {FloatData [40,40,40,40,40,40,40,40,40,40,40,40,40]}}
    )
    UsePositionOffset=false
    Write Material Random=true
   )
   &CameraOffset(Camera Offset Amount=150)
   &SystemLocation(MaskSpawnGroup=false, Offset=<0 0 150>)
  ]
  particleUpdateScript [
   &SkeletalMeshLocation(
    Bone / Socket Index=$Particles.SpawnGroup
    Bone Sampling Mode=::NewEnumerator4
    MaskSpawnGroup=false
    Skeletal Mesh=DataInterface() {propertyValues {
      DefaultMesh=null
      ExcludeBoneName=root
      OptionalComponentClass=null
      PreviewMesh=&M_Med_Soldier_04_Preview
      SkinningMode=SkinOnTheFly
      SoftSourceActor=null
      SourceMode=Default
      UvSetIndex=0
      WholeMeshLOD=-1
      bExcludeBone=false
      bReadDeformedGeometry=true
      bRequireCurrentFrameData=true
      ComponentTags []
      FilteredBones [head,upperarm_l,upperarm_r,lowerarm_l,lowerarm_r,hand_l,hand_r,thigh_l,thigh_r,calf_l,calf_r,foot_l,foot_r]
      FilteredSockets []
      MeshUserParameter {parameter {name=None, typeDefHandle {registeredTypeIndex=1}, varData []}}
      SamplingRegions []
     }}
   )
   &ParticleState()
   &ScaleSpriteSize(Scale Sprite Size Mode=::NewEnumerator0, Uniform Scale Factor=5)
   &ScaleColor(
    Scale Alpha=&FloatFromCurve(FloatCurve=DataInterface() {propertyValues {Curve {keys [<at 0 1>,<at 0.5 1>,<at 1 0>]}}})
    ScaleA=true
    ScaleRGB=true
   )
   &DynamicMaterialParameters(
    Index 0 Param 1=1
    Index 0 Param 2=&FloatFromCurve(FloatCurve=DataInterface() {propertyValues {Curve {keys [<at 0 0 RCIM_Constant>, {tangentMode=RCTM_User, time=0.7, value=0}]}}})
    Module.Param0WriteEnabled=true
    Module.Param1WriteEnabled=true
    Module.Param2WriteEnabled=true
    Module.Param3WriteEnabled=true
   )
  ]
  rendererClasses [&NiagaraSpriteRendererProperties]
  renderers [
   {rendererClass=&NiagaraSpriteRendererProperties}
   {rendererClass=&NiagaraSpriteRendererProperties, rendererIndex=1}
  ]
 }
 Minimal {
  emitterUpdateScript [
   &EmitterState()
   &SpawnBurst_Instantaneous(Use Spawn Probability=false, UseLoopCountLimit=false)
  ]
  particleSpawnScript [
   &InitializeParticle(
    Color=<rgba 0 2.6153 3 1>
    Lifetime=0.3
    Position Offset=<0 0 10>
    Sprite Size Mode=::NewEnumerator3
    Uniform Sprite Size=100
    UsePositionOffset=true
    Write Material Random=true
   )
   &CameraOffset(Camera Offset Amount=80)
  ]
  particleUpdateScript [
   &ParticleState()
   &ScaleSpriteSize(Uniform Curve Sprite Scale=DataInterface() {propertyValues {Curve {keys [{arriveTangent=2.2828, interpMode=RCIM_Cubic, leaveTangent=2.2828, tangentMode=RCTM_User, time=0, value=0}, <at 1 1 RCIM_Cubic>]}}})
   &ScaleColor(
    Scale Alpha=&FloatFromCurve(FloatCurve=DataInterface() {propertyValues {Curve {keys [<at 0 1>,<at 1 0>]}}})
    ScaleA=true
    ScaleRGB=true
   )
  ]
  rendererClasses [&NiagaraSpriteRendererProperties]
  renderers [{rendererClass=&NiagaraSpriteRendererProperties}]
 }
 _Frontend_Dark_flash {
  emitterUpdateScript [
   &EmitterState(Life Cycle Mode=::NewEnumerator1, Loop Behavior=::NewEnumerator1, Loop Duration=1, UseLoopDelay=false)
   &SpawnBurst_Instantaneous(
    Spawn Count=&MakeIntFromFloat(Float=&FN_FrontEndFloatModifier(Gameplay=0, Front End=13))
    Use Spawn Probability=false
    UseLoopCountLimit=false
   )
  ]
  particleSpawnScript [
   &FN_SetSpawnGroup(Number of Spawn Groups=13)
   &InitializeParticle(
    Color Mode=::NewEnumerator0
    Lifetime=0.2
    Sprite Rotation Angle=&MaskFloatBySpawnGroup(Masked Float=0.01, Passthrough Float=0, Spawn Group Mask=11)
    Sprite Rotation Mode=::NewEnumerator1
    Sprite Size=&SelectVector2DFromArray(
     Direct Array Index=&SpawnGroup()
     Vector 2D Selection Array=DataInterface() {propertyValues {FloatData [<70 70>, <70 70>, <70 70>, <70 70>, <70 70>, <70 70>, <70 70>, <70 70>, <70 70>, <40 40>, <40 40>, <40 50>, <50 50>]}}
    )
    Sprite Size Mode=::NewEnumerator4
    UsePositionOffset=false
    Write Material Random=true
   )
   &CameraOffset(Camera Offset Amount=50)
   &SystemLocation(MaskSpawnGroup=false, Offset=<0 0 150>)
  ]
  particleUpdateScript [
   &SkeletalMeshLocation(
    Bone / Socket Index=$Particles.SpawnGroup
    Bone Sampling Mode=::NewEnumerator4
    MaskSpawnGroup=false
    Sampled Position Offset=&MaskVector(Mask Source=0, Mask Parameter=?, Mask=11, Masked Vector=<-3 0 2>, Passthrough Vector=<0 0 0>)
    Skeletal Mesh=DataInterface() {propertyValues {
      DefaultMesh=null
      ExcludeBoneName=root
      OptionalComponentClass=null
      PreviewMesh=&M_Med_Soldier_04_Preview
      SkinningMode=SkinOnTheFly
      SoftSourceActor=null
      SourceMode=Default
      UvSetIndex=0
      WholeMeshLOD=-1
      bExcludeBone=false
      bReadDeformedGeometry=true
      bRequireCurrentFrameData=true
      ComponentTags []
      FilteredBones [head,upperarm_l,upperarm_r,lowerarm_l,lowerarm_r,hand_l,hand_r,thigh_l,thigh_r,calf_l,calf_r,foot_l,foot_r]
      FilteredSockets []
      MeshUserParameter {parameter {name=None, typeDefHandle {registeredTypeIndex=1}, varData []}}
      SamplingRegions []
     }}
   )
   &ParticleState()
   &ScaleColor(
    Scale Alpha=&FloatFromCurve(FloatCurve=DataInterface() {propertyValues {Curve {keys [<at 0 1>,<at 0.5 1>,<at 1 0>]}}})
    ScaleA=true
    ScaleRGB=true
   )
   &DynamicMaterialParameters(
    Index 0 Param 1=1
    Index 0 Param 2=&FloatFromCurve(FloatCurve=DataInterface() {propertyValues {Curve {keys [<at 0 0 RCIM_Constant>, {tangentMode=RCTM_User, time=0.7, value=1}]}}})
    Module.Param0WriteEnabled=true
    Module.Param1WriteEnabled=true
    Module.Param2WriteEnabled=true
    Module.Param3WriteEnabled=true
   )
  ]
  rendererClasses [&NiagaraSpriteRendererProperties]
  renderers [{rendererClass=&NiagaraSpriteRendererProperties}]
 }
}
summary {
 systemName=NS_Emote_KneelReedy_Flash
 emitters [
  {emitterName=Minimal, rendererClasses [&NiagaraSpriteRendererProperties]}
  {emitterName=Dark_flash, rendererClasses [&NiagaraSpriteRendererProperties]}
  {bEnabled=false, emitterName=Debug, rendererClasses [&NiagaraSpriteRendererProperties]}
  {emitterName=_Frontend_Dark_flash, rendererClasses [&NiagaraSpriteRendererProperties]}
 ]
 userVariables [
  User.bIsFrontEnd {type=NiagaraBool}
  User.bIsFrontEndPreview {type=NiagaraBool}
  User.FX Position {type=Vector3f, description="", defaultValue {struct=Vector3f, value=<0 0 0>}}
 ]
}
